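# Physics asset for the Holiday Fish gadget mesh, giving it capsule bodies for its spine and scarf bones.
# The Anim bone anchors two-link front and back spine chains.
# Inner and outer two-link scarf chains hang from the first front spine body.
# Every joint shares the same tight swing and twist limits, so the chains can flex only slightly.
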PhysicsAsset /Game/Gadgets/Assets/Fish/Holiday_Fish/Meshes/Holiday_Fish_Physics
bodies {
 Anim [{shapeType=Capsule, length=0.505, radius=0.505, rotation=<rot 0 -90 0>}]
 dyn_spineBack_01 [{shapeType=Capsule, length=9.5145, radius=3.4166, center=<3.0522 0 0>, rotation=<rot 90 0 0>}]
 dyn_spineBack_02 [{shapeType=Capsule, length=9.5145, radius=3.4166, center=<3.0522 0 0>, rotation=<rot 90 0 0>}]
 dyn_spineFront_01 [{shapeType=Capsule, length=9.5145, radius=3.4166, center=<2.3147 0 0>, rotation=<rot 90 0 0>}]
 dyn_spineFront_02 [{shapeType=Capsule, length=9.5145, radius=3.4166, center=<2.3147 0 0>, rotation=<rot 90 0 0>}]
 scarfIN_01 [{shapeType=Capsule, length=3.0019, radius=1.5936, center=<1.8938 0 0.0001>, rotation=<rot 90 0 0>}]
 scarfIN_02 [{shapeType=Capsule, length=3.7097, radius=2.5264, center=<1.8938 0 0.0001>, rotation=<rot 90 0 0>}]
 scarfOUT_01 [{shapeType=Capsule, length=3.2188, radius=1.4795, center=<1.8938 0 0.0001>, rotation=<rot 90 0 0>}]
 scarfOUT_02 [{shapeType=Capsule, length=4.0156, radius=2.8058, center=<1.8938 0 0.0001>, rotation=<rot 90 0 0>}]
}
constraints [
 {bone1Name=dyn_spineFront_01, bone2Name=Anim, swing1LimitDegrees=12, swing2LimitDegrees=12, twistLimitDegrees=12}
 {bone1Name=dyn_spineFront_02, bone2Name=dyn_spineFront_01, swing1LimitDegrees=12, swing2LimitDegrees=12, twistLimitDegrees=12}
 {bone1Name=dyn_spineBack_01, bone2Name=Anim, swing1LimitDegrees=12, swing2LimitDegrees=12, twistLimitDegrees=12}
 {bone1Name=dyn_spineBack_02, bone2Name=dyn_spineBack_01, swing1LimitDegrees=12, swing2LimitDegrees=12, twistLimitDegrees=12}
 {bone1Name=scarfOUT_01, bone2Name=dyn_spineFront_01, swing1LimitDegrees=12, swing2LimitDegrees=12, twistLimitDegrees=12}
 {bone1Name=scarfOUT_02, bone2Name=scarfOUT_01, swing1LimitDegrees=12, swing2LimitDegrees=12, twistLimitDegrees=12}
 {bone1Name=scarfIN_01, bone2Name=dyn_spineFront_01, swing1LimitDegrees=12, swing2LimitDegrees=12, twistLimitDegrees=12}
 {bone1Name=scarfIN_02, bone2Name=scarfIN_01, swing1LimitDegrees=12, swing2LimitDegrees=12, twistLimitDegrees=12}
]
